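# Physics asset for the face accessory part of the medium male Progressive Jonesy body, covering three dynamic hair bones and the head.
# Each body is a single capsule, with identical shapes on the hair bones and a larger one on the head.
# The hair bones each constrain directly to the head rather than forming a chain.
# Small swing and twist limit angles are set, with narrower swing on the third hair bone.
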PhysicsAsset /Game/Characters/Player/Male/Medium/Bodies/M_MED_Progressive_Jonesy/Meshes/Parts/M_MED_Progressive_Jonesy_faceAcc_PHYS
bodies {
 dyn_hair_01 [{shapeType=Capsule, length=2.264, radius=1.3812, center=<1 0 0>, rotation=<rot 90 0 0>}]
 dyn_hair_02 [{shapeType=Capsule, length=2.264, radius=1.3812, center=<1 0 0>, rotation=<rot 90 0 0>}]
 dyn_hair_03 [{shapeType=Capsule, length=2.264, radius=1.3812, center=<1 0 0>, rotation=<rot 90 0 0>}]
 head [{shapeType=Capsule, length=3.5122, radius=3.1752}]
}
constraints [
 {bone1Name=dyn_hair_02, bone2Name=head, swing1LimitDegrees=15, swing2LimitDegrees=15, twistLimitDegrees=5}
 {bone1Name=dyn_hair_01, bone2Name=head, swing1LimitDegrees=15, swing2LimitDegrees=15, twistLimitDegrees=5}
 {bone1Name=dyn_hair_03, bone2Name=head, swing1LimitDegrees=10, swing2LimitDegrees=10, twistLimitDegrees=5}
]
